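# Material function for the MarchTreatCup female body cosmetic that adds audio-reactive glow to the emissive channel of incoming material attributes.
# Frontend-loop audio amplitude, scaled by an Is Reactive parameter and raised to an exponent, serves as the reactivity signal.
# An FX mask splits into cloth, detail and gradient masks alongside Fresnel-based body and detail glows and a bow glow.
# Pre-skinned local X yields a left-side mask, and IsFaceAccessory switches the emissive blend.
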MaterialFunction /BRCosmetics/Characters/Player/Female/Medium/Bodies/F_MED_MarchTreatCup/Materials/MF_MarchTreatCup_Reactivity
graph {nodes [
  Constant() {r=1}
  n1 = GetMaterialAttributes(None=FunctionInput() {inputType=FunctionInput_MaterialAttributes}) {attributeGetTypes [#1]}
  FunctionOutput(None=SetMaterialAttributes!(MaterialAttributes=@n1.MaterialAttributes, EmissiveColor=NamedRerouteUsage()) {attributeSetTypes [#1]})
  MaterialFunctionCall(Speed=Constant() {r=2}) {materialFunction=&DebugTimeSine}
  NamedRerouteDeclaration(None=@n1.EmissiveColor) {name=InEmissive, nodeColor=<rgba 0 0.625 1 1>}
  n2 = VectorParameter() {parameterName=ReactivityBowRightColor, defaultValue=<rgba 0 0.3586 5 1>}
  n3 = VectorParameter() {parameterName=ReactivityBowLeftColor, defaultValue=<rgba 2 1.0562 0.1146 1>}
  n4 = NamedRerouteUsage()
  n5 = ScalarParameter() {defaultValue=1.5, parameterName=ExponentIn}
  n6 = Power(Base=NamedRerouteUsage(), Exp=LinearInterpolate(A=@n5, B=Multiply(A=@n5) {constB=0.2}, Alpha=NamedRerouteUsage()))
  n7 = ScalarParameter() {defaultValue=0.8, parameterName=BowEmissiveIntensity}
  n8 = LinearInterpolate(A=@n7, B=Multiply(A=@n7) {constB=1.3}, Alpha=NamedRerouteUsage())
  NamedRerouteDeclaration(None=LinearInterpolate(A=NamedRerouteUsage(), B=LinearInterpolate(A=@n2.RGB, B=@n3.RGB, Alpha=@n4), Alpha=Multiply(A=@n6, B=@n8))) {name=BowReactivity, nodeColor=<rgba 0 1 0.2734 1>}
  n9 = TextureSample() {
   samplerType=SAMPLERTYPE_Masks
   texture=/BRCosmetics/Characters/Player/Female/Medium/Bodies/F_MED_MarchTreatCup/Textures/T_MarchTreatCup_Body_FX
  }
  n10 = MaterialFunctionCall(Color=NamedRerouteDeclaration(None=FunctionInput(Preview=@n9.RGB) {inputName=FXMask, sortPriority=1}) {name=InFXMask, nodeColor=<rgba 0.9453 1 0 1>}) {materialFunction=&SplitComponents}
  NamedRerouteDeclaration(None=@n10.R) {name=WhiteClothMask, nodeColor=<rgba 0.7344 0 1 1>}
  NamedRerouteDeclaration(None=@n10.G) {name=DetailsMask, nodeColor=<rgba 0 1 0.4141 1>}
  NamedRerouteDeclaration(None=@n10.B) {name=GradientAccents, nodeColor=<rgba 1 0.0938 0 1>}
  n11 = Multiply(
   A=MaterialFunctionCall!(InAmplitude, InSpectrum, CustomFakeSpectrum, CustomFakeAmplitude) {materialFunction=/BRCosmetics/Global/MaterialFunctions/AudioReactivity/MF_AudioReactivity_FrontendLoop}.Amplitude
   B=ScalarParameter() {defaultValue=1, parameterName=Is Reactive}
  )
  NamedRerouteDeclaration(None=Power(
    Base=NamedRerouteDeclaration(None=@n11) {name=RawAmplitude, nodeColor=<rgba 0 1 0.4141 1>}
    Exp=ScalarParameter() {defaultValue=4, parameterName=AmplitudeExponent}
   ) {constExponent=4}) {name=Amplitude, nodeColor=<rgba 0 0.2266 1 1>}
  n12 = Add(A=NamedRerouteUsage(), B=NamedRerouteUsage())
  n13 = NamedRerouteUsage()
  n14 = ScalarParameter() {parameterName=IsFaceAccessory}
  n15 = NamedRerouteUsage()
  n16 = ScalarParameter() {defaultValue=0.2, parameterName=B_1}
  NamedRerouteDeclaration(None=LinearInterpolate(A=NamedRerouteUsage(), B=LinearInterpolate(A=@n12, B=@n13, Alpha=@n14), Alpha=Power(Base=@n15, Exp=@n16))) {name=OutEmissive, nodeColor=<rgba 0.5469 1 0 1>}
  n17 = VectorParameter() {parameterName=ReactivityBodyColor, defaultValue=<rgba 0.3178 0 0.8438 1>}
  n18 = ScalarParameter() {defaultValue=0.3, parameterName=ReactivityFresnel}
  n19 = ScalarParameter() {defaultValue=0.25, parameterName=B}
  n20 = Fresnel(ExponentIn=LinearInterpolate(A=@n18, B=Multiply(A=@n18, B=@n19) {constB=0.1}, Alpha=NamedRerouteUsage())) {baseReflectFraction=0}
  n21 = Add(A=Multiply(A=SmoothStep(Value=@n20) {constMin=0.9}, B=ScalarParameter() {defaultValue=5, parameterName=BodyRimIntensity}) {constB=15})
  n22 = NamedRerouteUsage()
  n23 = ScalarParameter() {defaultValue=0.4, parameterName=ReactivityFresnelMin}
  n24 = ScalarParameter() {defaultValue=1, parameterName=ReactivityFresnelMax}
  n25 = NamedRerouteUsage()
  n26 = ScalarParameter() {defaultValue=0.1442, parameterName=EffectFresnel}
  n27 = Multiply(
   A=Max(A=SmoothStep(Min=@n23, Max=@n24, Value=@n20) {constMin=0.2}, B=Power(Base=@n25))
   B=SmoothStep(
    Min=ScalarParameter() {defaultValue=0.5, parameterName=EffectFresnelMin}
    Value=Fresnel(ExponentIn=@n26) {baseReflectFraction=0, exponent=2.5}
   ) {constMin=0.8}
  )
  NamedRerouteDeclaration(None=LinearInterpolate(A=NamedRerouteUsage(), B=Multiply(A=@n17.RGB, B=@n21), Alpha=Multiply(A=@n22, B=@n27))) {name=FresnelReactivity, nodeColor=<rgba 0.0312 1 0 1>}
  n28 = ScalarParameter() {defaultValue=1.25, parameterName=DetailFresnel}
  n29 = Fresnel(ExponentIn=LinearInterpolate(A=@n28, B=Multiply(A=@n28) {constB=0.5}, Alpha=NamedRerouteUsage())) {baseReflectFraction=0, exponent=2}
  n30 = Add(A=Multiply(A=SmoothStep(Value=@n29) {constMax=0.75}, B=ScalarParameter() {defaultValue=3, parameterName=DetailRimIntensity}), B=@n29)
  n31 = VectorParameter() {parameterName=ReactivityAccentColor, defaultValue=<rgba 0 0.1845 1 1>}
  NamedRerouteDeclaration(None=Multiply(A=Multiply(A=@n30, B=@n31.RGB), B=NamedRerouteUsage())) {name=DetailReactivity, nodeColor=<rgba 1 0 0.3516 1>}
  n32 = Ceil(None=ComponentMask(None=LocalPosition() {localOrigin=InstancePreSkinning}.XYZ) {r=true})
  NamedRerouteDeclaration(None=VertexInterpolator(VS=Saturate(None=@n32)).PS) {name=LeftMask, nodeColor=<rgba 0 0.2031 1 1>}
  Saturate(None=Multiply(B=@n16))
  TextureSample() {
   samplerType=SAMPLERTYPE_Masks
   texture=/BRCosmetics/Characters/Player/Female/Medium/Bodies/F_MED_MarchTreatCup/Textures/T_MarchTreatCup_FaceAcc_FX
  }
 ]}
