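# Music-reactive Niagara body effect for the MarchTreatCup female character cosmetic, layering auras, decal sprites, line meshes, hearts and a body glow.
# Particles follow skeletal bones, mostly using partition-indexed arrays for bone, offset and size.
# Event audio amplitude and MusicReactivity gate spawning and drive brightness, alongside distance-based LOD fades and front-end preview overrides.
# Hidden hand, foot and head flags zero affected particle scales; the glow spawns only for local player or preview.
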NiagaraSystem /BRCosmetics/Characters/Player/Female/Medium/Bodies/F_MED_MarchTreatCup/Meshes/FX/NS_Character_MarchTreatCup_Body
dependencies {
 usedDataInterfaces [
  &NiagaraDataInterfaceArrayFloat
  &NiagaraDataInterfaceArrayFloat3
  &NiagaraDataInterfaceCurve
  &NiagaraDataInterfaceArrayInt32
  &NiagaraDataInterfaceColorCurve
  &NiagaraDataInterfaceSpriteRendererInfo
  &NiagaraDataInterfaceSkeletalMesh
 ]
 usedDynamicInputs []
 usedModules [
  &self:SystemSpawnScript.SystemScriptSource.SystemScriptGraph.NiagaraNodeAssignment_0.SetVariables_#1
  &self:SystemSpawnScript.SystemScriptSource.SystemScriptGraph.NiagaraNodeAssignment_1.SetVariables_#2
  &self:SystemSpawnScript.SystemScriptSource.SystemScriptGraph.NiagaraNodeAssignment_2.SetVariables_#3
  &SystemState
  &InterpolateOverTime
 ]
}
emitters {
 Aura_small001 {
  emitterUpdateScript [
   &EmitterState(Life Cycle Mode=::NewEnumerator1, Loop Duration=12.9, UseLoopDelay=false)
   &SpawnRate(
    Spawn Probability=&MakeCustomFloatFromBool(
     Bool=$User.bIsFrontEndPreview
     True Float=1
     False Float=&Lerp_Float(
      A=&MakeCustomFloatFromBool(
       Bool=&SetBoolByFloatComparison(A=$NPC.FortniteNPC.EventAudio_Amplitude, B=0, Comparison Type=::NewEnumerator4)
       True Float=0
       False Float=1
      )
      B=1
      Alpha=$System.MusicReactivity
     )
    )
    SpawnRate=10
    Use Spawn Probability=true
   )
  ]
  particleSpawnScript [
   &PartitionParticles(Number Of Partitions=12)
   &InitializeParticle(
    Color=<rgba 0.6771 1.1181 2 2>
    Lifetime=2
    Sprite Size Mode=::NewEnumerator3
    Sprite UV Mode=::NewEnumerator1
    Uniform Sprite Size=&SelectFloatFromArray(
     Direct Array Index=$Particles.PartitionParticles.Partition
     Float Selection Array=DataInterface() {propertyValues {FloatData [70,70,70,50,50,50,50,80,80,80,0,0,70]}}
    )
    UsePositionOffset=false
    Write Material Random=true
   )
   &CameraOffset(Camera Offset Amount=10)
  ]
  particleUpdateScript [
   &ParticleState()
   &SkeletalMeshLocation(
    Bone / Socket Index=$Particles.PartitionParticles.Partition
    Bone Sampling Mode=::NewEnumerator4
    MaskSpawnGroup=false
    Sampled Position Offset=&SelectVectorFromArray(
     Direct Array Index=$Particles.PartitionParticles.Partition
     Vector Selection Array=DataInterface() {propertyValues {FloatData [<0 0 0>, <0 0 0>, <0 0 0>, <0 0 0>, <0 0 0>, <5 0 0>, <-5 0 0>, <0 0 0>, <0 0 0>, <0 0 0>, <0 0 0>, <0 0 0>, <0 0 0>]}}
    )
    Skeletal Mesh=$System.Skeletal Mesh Reader
   )
   &ScaleSpriteSize(
    Scale Factor=&RandomRangeVector2D(Minimum=<0.5 0.6>, Maximum=<0.55 0.65>)
    Scale Sprite Size Mode=::NewEnumerator2
   ) {module=ScaleSpriteSize001}
   &ScaleColor(
    Scale Alpha=&MakeCustomFloatFromBool(Bool=$User.bIsFrontEndPreview, True Float=1.5, False Float=&Multiply_Float(A=3, B=$System.FNLODDistanceRemap))
    Scale RGB=&Multiply_VectorByFloat(Vector=<1 1 1>, Float=0.7)
    ScaleA=true
    ScaleRGB=true
   ) {module=ScaleColor001}
   &ScaleColor(
    Scale Alpha=&FloatFromCurve(
     FloatCurve=DataInterface() {propertyValues {Curve {keys [
         {tangentMode=RCTM_User, time=0, value=0}
         <at 0.2 1>
         {arriveTangent=-0.0405, interpMode=RCIM_Cubic, leaveTangent=-0.0405, tangentMode=RCTM_User, time=0.4, value=1}
         {arriveTangent=-1.4489, interpMode=RCIM_Cubic, leaveTangent=-1.4489, tangentMode=RCTM_User, time=1, value=0}
        ]}}}
     Scale Curve=&MakeCustomFloatFromBool(
      Bool=$User.bIsFrontEndPreview
      True Float=&ClampFloat(Float=$NPC.FortniteNPC.EventAudio_Amplitude, Min=0.3, Max=1.3)
      False Float=&Lerp_Float(
       A=&Multiply_Float(
        A=&Lerp_Float(A=0, B=1, Alpha=$System.InterpolateOverTime.InterpolateOverTime_Float.Moving Average)
        B=&ClampFloat(Float=$NPC.FortniteNPC.EventAudio_Amplitude, Min=0.9, Max=1.3)
       )
       B=1
       Alpha=$System.MusicReactivity
      )
     )
    )
    Scale RGB=&MakeVectorFromLinearColorRGB(Color=<rgba 5 1.4062 2.4935 1>)
    ScaleA=true
    ScaleRGB=true
   )
   &DynamicMaterialParameters(
    Index 0 Param 1=&RandomRangeFloat(Minimum=0.5, Maximum=1.5)
    Module.Index0Float0Write=true
    Module.Index0Float1Write=true
    Module.Index0Float2Write=true
    Module.Index0Float3Write=true
   )
   &ScaleSpriteSize(
    Scale Sprite Size Mode=::NewEnumerator0
    Uniform Scale Factor=&MakeCustomFloatFromBool(
     Bool=$User.bisHandHidden
     True Float=&SelectFloatFromArray(
      Direct Array Index=$Particles.PartitionParticles.Partition
      Float Selection Array=DataInterface() {propertyValues {FloatData [1,1,0,0,0,0,1,1,1,1,1,1]}}
     )
     False Float=&SelectFloatFromArray(
      Direct Array Index=$Particles.PartitionParticles.Partition
      Float Selection Array=DataInterface() {propertyValues {FloatData [1,1,1,1,1,1,1,1,1,1,1,1]}}
     )
    )
   )
   &ScaleSpriteSize(
    Scale Sprite Size Mode=::NewEnumerator0
    Uniform Scale Factor=&MakeCustomFloatFromBool(
     Bool=$User.bisFootHidden
     True Float=&SelectFloatFromArray(
      Direct Array Index=$Particles.PartitionParticles.Partition
      Float Selection Array=DataInterface() {propertyValues {FloatData [1,1,1,1,1,1,1,1,0,0,0,0]}}
     )
     False Float=&SelectFloatFromArray(
      Direct Array Index=$Particles.PartitionParticles.Partition
      Float Selection Array=DataInterface() {propertyValues {FloatData [1,1,1,1,1,1,1,1,1,1,1,1]}}
     )
    )
   ) {module=ScaleSpriteSize002}
   &ScaleColor(
    Scale Alpha=&MakeCustomFloatFromBool(
     Bool=$User.bIsFrontEndPreview
     True Float=&MakeCustomFloatFromBool(
      Bool=&SetBoolByFloatComparison(A=$NPC.FortniteNPC.EventAudio_Amplitude, B=0, Comparison Type=::NewEnumerator4)
      True Float=&Lerp_Float(A=0, B=1, Alpha=$User.EnableFrontendFakeAudioLoop)
      False Float=1
     )
     False Float=1
    )
    ScaleA=true
    ScaleRGB=true
   ) {module=ScaleColor002}
  ]
  rendererClasses [&NiagaraSpriteRendererProperties]
  renderers [{rendererClass=&NiagaraSpriteRendererProperties}]
 }
 Aura_small003 {
  emitterUpdateScript [
   &EmitterState(Life Cycle Mode=::NewEnumerator1, Loop Duration=12.9, UseLoopDelay=false)
   &SpawnRate(
    Spawn Probability=&MakeCustomFloatFromBool(
     Bool=$User.bIsFrontEndPreview
     True Float=1
     False Float=&MakeCustomFloatFromBool(
      Bool=&SetBoolByFloatComparison(A=$NPC.FortniteNPC.EventAudio_Amplitude, B=0, Comparison Type=::NewEnumerator4)
      True Float=&Lerp_Float(A=0, B=1, Alpha=$System.MusicReactivity)
      False Float=1
     )
    )
    SpawnRate=60
    Use Spawn Probability=true
   )
  ]
  particleSpawnScript [
   &PartitionParticles(Number Of Partitions=10)
   &InitializeParticle(
    Color=&MakeCustomLinearColorFromBool(
     Bool=&RandomBool()
     True Color=&MakeCustomLinearColorFromBool(Bool=&RandomBool(), True Color=<rgba 0.1147 0.2888 1 1>, False Color=<rgba 0.5065 0.1086 1 1>)
     False Color=<rgba 1 0.1146 0.5159 1>
    )
    Lifetime Max=0.5
    Lifetime Min=0.3
    Lifetime Mode=::NewEnumerator1
    Mesh Scale Max=<1.2 1.2 12>
    Mesh Scale Min=<0.6 0.6 4>
    Mesh Scale Mode=::NewEnumerator2
    Position Offset=<0 0 100>
    UsePositionOffset=true
    Write Material Random=true
   )
   &CameraOffset(Camera Offset Amount=5)
   &ShapeLocation(
    Apply To Particle Position=::NewEnumerator1
    MaskSpawnGroup=false
    Offset=&SelectVectorFromArray(
     Direct Array Index=$Particles.PartitionParticles.Partition
     Vector Selection Array=DataInterface() {propertyValues {FloatData [<0 0 0>, <0 0 0>, <-8 0 0>, <8 0 0>, <0 -5 -5>, <5 5 5>, <0 0 0>, <0 0 0>]}}
    )
    Offset Mode=::NewEnumerator0
    Override Seed=false
    Sphere Radius=5
    UseOwnerScale=true
   )
   &AddVelocity(Velocity=<0 0 -10>)
  ]
  particleUpdateScript [
   &ParticleState()
   &SkeletalMeshLocation(
    Bone / Socket Index=&SelectIntFromArray(Direct Array Index=$Particles.PartitionParticles.Partition, Int Selection Array=DataInterface() {propertyValues {IntData [2,3,4,5,6,7,8,9]}})
    Bone Sampling Mode=::NewEnumerator4
    MaskSpawnGroup=false
    Orientation Sampling=::NewEnumerator1
    Sampled Position Offset=$Particles.ShapeLocation.ShapeVector
    Skeletal Mesh=$System.Skeletal Mesh Reader
   )
   &ScaleColor(Scale Alpha=$System.FNLODDistanceRemap, ScaleA=true, ScaleRGB=true) {module=ScaleColor001}
   &ScaleColor(
    Scale Alpha=&Multiply_Float(
     A=&FloatFromCurve(
      FloatCurve=DataInterface() {propertyValues {Curve {keys [<at 0 0>,<at 0.1 1>,<at 0.9 1>,<at 1 0>]}}}
      Scale Curve=&MakeCustomFloatFromBool(
       Bool=$User.bIsFrontEndPreview
       True Float=&RandomRangeFloat(Minimum=0.6, Maximum=1)
       False Float=&RandomRangeFloat(Minimum=1, Maximum=2)
      )
     )
     B=&MakeCustomFloatFromBool(
      Bool=$User.bIsFrontEndPreview
      True Float=&ClampFloat(Float=$NPC.FortniteNPC.EventAudio_Amplitude, Min=0.5, Max=1.5)
      False Float=&Lerp_Float(
       A=&Multiply_Float(
        A=&Lerp_Float(A=0, B=1, Alpha=$System.InterpolateOverTime.InterpolateOverTime_Float.Moving Average)
        B=&ClampFloat(Float=$NPC.FortniteNPC.EventAudio_Amplitude, Min=0.5)
       )
       B=1
       Alpha=$System.MusicReactivity
      )
     )
    )
    Scale RGB=&Multiply_VectorByFloat(
     Vector=&MakeVectorFromLinearColorRGB(Color=&ColorFromCurve(ColorCurve=DataInterface() {propertyValues {
         AlphaCurve {keys []}
         BlueCurve {keys [<at 0 3>,<at 0.1326 3>,<at 0.2869 0>,<at 0.5108 3>,<at 0.8097 5>]}
         GreenCurve {keys [<at 0 2.8212>,<at 0.1326 0>,<at 0.2869 1.2188>,<at 0.5108 0.4551>,<at 0.8097 0.9929>]}
         RedCurve {keys [<at 0 0>,<at 0.1326 1.479>,<at 0.2869 3>,<at 0.5108 2.7742>,<at 0.8097 0>]}
        }}))
     Float=&FloatFromCurve(FloatCurve=DataInterface() {propertyValues {Curve {keys [<at 0 1 RCIM_Cubic>,<at 1 0.6 RCIM_Cubic>]}}})
    )
    ScaleA=true
    ScaleRGB=true
   )
   &ScaleMeshSize(Scale Factor=&Multiply_VectorByFloat(
     Vector=&MakeVector(X=1, Y=1, Z=&Lerp_Float(A=0.8, B=1.5, Alpha=$NPC.FortniteNPC.EventAudio_Amplitude))
     Float=&Multiply_Float(
      A=0.1
      B=&FloatFromCurve(
       FloatCurve=DataInterface() {propertyValues {Curve {keys [
           {arriveTangent=4.1785, time=0, value=1}
           <at 0.5 1>
           {arriveTangent=-0.0405, interpMode=RCIM_Cubic, leaveTangent=-0.0405, tangentMode=RCTM_User, time=1, value=1}
          ]}}}
       Scale Curve=&Multiply_Float(
        A=1
        B=&ClampFloat(
         Float=&FN_LOD_Distance_Remap(Start Distance=200, End Distance=400, Minimum Output=0.5, Maximum Output=1)
         Min=0.5
         Max=0.8
        )
       )
      )
     )
    )) {module=ScaleMeshSize001}
   &Drag(Drag=4, Use Linear Drag=true, Use Rotational Drag=true)
   &SolveForcesAndVelocity(Clamp Velocity=false, Limit Acceleration=false)
   &ScaleMeshSize(Scale Factor=&VectorFromFloat(Value=&MakeCustomFloatFromBool(
      Bool=$User.bisHandHidden
      True Float=&SelectFloatFromArray(
       Direct Array Index=$Particles.PartitionParticles.Partition
       Float Selection Array=DataInterface() {propertyValues {FloatData [0,0,0,0,1,1,1,1]}}
      )
      False Float=&SelectFloatFromArray(
       Direct Array Index=$Particles.PartitionParticles.Partition
       Float Selection Array=DataInterface() {propertyValues {FloatData [1,1,1,1,1,1,1,1]}}
      )
     ))) {module=ScaleMeshSize002}
   &ScaleMeshSize(Scale Factor=&VectorFromFloat(Value=&MakeCustomFloatFromBool(
      Bool=$User.bisFootHidden
      True Float=&SelectFloatFromArray(
       Direct Array Index=$Particles.PartitionParticles.Partition
       Float Selection Array=DataInterface() {propertyValues {FloatData [1,1,1,1,1,1,0,0]}}
      )
      False Float=&SelectFloatFromArray(
       Direct Array Index=$Particles.PartitionParticles.Partition
       Float Selection Array=DataInterface() {propertyValues {FloatData [1,1,1,1,1,1,1,1]}}
      )
     ))) {module=ScaleMeshSize003}
   &ScaleColor(
    Scale Alpha=&MakeCustomFloatFromBool(
     Bool=$User.bIsFrontEndPreview
     True Float=&MakeCustomFloatFromBool(
      Bool=&SetBoolByFloatComparison(A=$NPC.FortniteNPC.EventAudio_Amplitude, B=0, Comparison Type=::NewEnumerator4)
      True Float=&Lerp_Float(A=0, B=1, Alpha=$User.EnableFrontendFakeAudioLoop)
      False Float=1
     )
     False Float=1
    )
    ScaleA=true
    ScaleRGB=true
   ) {module=ScaleColor002}
  ]
  rendererClasses [&NiagaraMeshRendererProperties]
  renderers [{rendererClass=&NiagaraMeshRendererProperties, rendererIndex=0}]
 }
 Decal {
  emitterUpdateScript [
   &EmitterState(Life Cycle Mode=::NewEnumerator1, Loop Duration=1, UseLoopDelay=false)
   &SpawnRate(
    Spawn Probability=&MakeCustomFloatFromBool(
     Bool=&SetBoolByFloatComparison(A=$NPC.FortniteNPC.EventAudio_Amplitude, B=0, Comparison Type=::NewEnumerator4)
     True Float=&Lerp_Float(A=0, B=1, Alpha=$System.MusicReactivity)
     False Float=1
    )
    SpawnRate=6
    Use Spawn Probability=true
   )
  ]
  particleSpawnScript [
   &PartitionParticles(Number Of Partitions=13)
   &InitializeParticle(
    Color=<rgba 3.6717 0.1823 5 2>
    Lifetime Mode=::NewEnumerator1
    Sprite Size Mode=::NewEnumerator3
    Sprite UV Mode=::NewEnumerator1
    Uniform Sprite Size=&SelectFloatFromArray(
     Direct Array Index=$Particles.PartitionParticles.Partition
     Float Selection Array=DataInterface() {propertyValues {FloatData [0,0,20,20,20,20,35,35,30,40,0,0,0]}}
    )
    UsePositionOffset=false
    Write Material Random=true
   )
   &CameraOffset(Camera Offset Amount=30)
  ]
  particleUpdateScript [
   &ParticleState()
   &SkeletalMeshLocation(
    Bone / Socket Index=$Particles.PartitionParticles.Partition
    Bone Sampling Mode=::NewEnumerator4
    MaskSpawnGroup=false
    Sampled Position Offset=&SelectVectorFromArray(
     Direct Array Index=$Particles.PartitionParticles.Partition
     Vector Selection Array=DataInterface() {propertyValues {FloatData [<0 0 0>, <0 0 0>, <0 0 5>, <0 -10 -5>, <0 0 0>, <5 0 0>, <-5 0 -10>, <15 0 10>, <-15 0 -10>, <10 0 10>, <20 0 20>, <0 0 0>, <0 0 0>]}}
    )
    Skeletal Mesh=$System.Skeletal Mesh Reader
   )
   &ScaleColor(
    Scale Alpha=&FloatFromCurve(
     FloatCurve=DataInterface() {propertyValues {Curve {keys [
         <at 0 0>
         <at 0.2 1>
         {arriveTangent=-0.0405, leaveTangent=-0.0405, time=0.5, value=1}
         {arriveTangent=-1.4489, leaveTangent=-1.4489, time=1, value=0}
        ]}}}
     Scale Curve=&Lerp_Float(
      A=&Multiply_Float(
       A=&Multiply_Float(
        A=&Lerp_Float(A=0, B=1.5, Alpha=$System.InterpolateOverTime.InterpolateOverTime_Float.Moving Average)
        B=&ClampFloat(Float=$NPC.FortniteNPC.EventAudio_Amplitude, Min=0.5)
       )
       B=$System.FNLODDistanceRemap
      )
      B=1
      Alpha=$System.MusicReactivity
     )
    )
    Scale RGB=&MakeVectorFromLinearColorRGB(Color=<rgba 1 1 1 1>)
    ScaleA=true
    ScaleRGB=true
   )
   &DynamicMaterialParameters(
    Index 0 Param 1=&RandomRangeFloat(Minimum=0.5, Maximum=1.5)
    Module.Index0Float0Write=true
    Module.Index0Float1Write=true
    Module.Index0Float2Write=true
    Module.Index0Float3Write=true
   )
   &ScaleSpriteSize(
    Scale Sprite Size Mode=::NewEnumerator0
    Uniform Scale Factor=&MakeCustomFloatFromBool(
     Bool=$User.bisHandHidden
     True Float=&SelectFloatFromArray(
      Direct Array Index=$Particles.PartitionParticles.Partition
      Float Selection Array=DataInterface() {propertyValues {FloatData [1,1,0,0,0,0,1,1,1,1,1,1,0]}}
     )
     False Float=&SelectFloatFromArray(
      Direct Array Index=$Particles.PartitionParticles.Partition
      Float Selection Array=DataInterface() {propertyValues {FloatData [1,1,1,1,1,1,1,1,1,1,1,1,0]}}
     )
    )
   )
   &ScaleSpriteSize(
    Scale Sprite Size Mode=::NewEnumerator0
    Uniform Scale Factor=&MakeCustomFloatFromBool(
     Bool=$User.bisFootHidden
     True Float=&SelectFloatFromArray(
      Direct Array Index=$Particles.PartitionParticles.Partition
      Float Selection Array=DataInterface() {propertyValues {FloatData [1,1,1,1,1,1,1,1,0,0,0,0,0]}}
     )
     False Float=&SelectFloatFromArray(
      Direct Array Index=$Particles.PartitionParticles.Partition
      Float Selection Array=DataInterface() {propertyValues {FloatData [1,1,1,1,1,1,1,1,1,1,1,1,0]}}
     )
    )
   ) {module=ScaleSpriteSize002}
  ]
  rendererClasses [&NiagaraSpriteRendererProperties]
  renderers [
   {rendererClass=&NiagaraSpriteRendererProperties}
   {rendererClass=&NiagaraSpriteRendererProperties, rendererIndex=1}
  ]
 }
 Glow_Body {
  emitterUpdateScript [
   /SetVariables(Emitter.GlowOutlineThickness=1) {isSetParameters=true}
   &EmitterState(Life Cycle Mode=::NewEnumerator1, Loop Duration=1, UseLoopDelay=false)
   &SpawnBurst_Instantaneous(
    Spawn Count=&MakeCustomIntFromBool(
     Bool=$User.bIsFrontEndPreview
     True INT=1
     False INT=&MakeCustomIntFromBool(Bool=$User.bIsLocalPlayer, True INT=1, False INT=0)
    )
    Spawn Probability=&MakeCustomFloatFromBool(
     Bool=&SetBoolByFloatComparison(A=$NPC.FortniteNPC.EventAudio_Amplitude, B=0, Comparison Type=::NewEnumerator4)
     True Float=&Lerp_Float(A=0, B=1, Alpha=$System.MusicReactivity)
     False Float=1
    )
    Use Spawn Probability=true
    UseLoopCountLimit=false
   )
  ]
  particleSpawnScript [
   &InitializeParticle(Lifetime=1.5, Mesh Scale=<1.2 1.2 4>, Mesh Scale Mode=::NewEnumerator4, UsePositionOffset=false, Write Material Random=true)
   &CameraOffset(Camera Offset Amount=20) {enabled=false}
  ]
  particleUpdateScript [
   &ParticleState()
   &ScaleMeshSize(Scale Factor=&VectorFromFloat(Value=&MakeCustomFloatFromBool(Bool=$User.bisHandHidden, True Float=0, False Float=1)))
   &ScaleMeshSize(Scale Factor=&VectorFromFloat(Value=&MakeCustomFloatFromBool(Bool=$User.bisFootHidden, True Float=0, False Float=1))) {module=ScaleMeshSize001}
   &ScaleMeshSize(Scale Factor=&VectorFromFloat(Value=&MakeCustomFloatFromBool(Bool=$User.bisHeadHidden, True Float=0, False Float=1))) {module=ScaleMeshSize002}
   &SkeletalMeshLocation(
    Bone Sampling Mode=::NewEnumerator4
    MaskSpawnGroup=false
    Orientation Sampling=::NewEnumerator1
    Sampled Position Offset=<0 0 40>
    Skeletal Mesh=$System.Skeletal Mesh
   ) {module=SkeletalMeshLocation001}
   &ScaleColor(
    Scale Alpha=&FloatFromCurve(
     FloatCurve=DataInterface() {propertyValues {Curve {keys [<at 0 0>,<at 0.1 1>,<at 0.9 1>,<at 1 0>]}}}
     Scale Curve=&MakeCustomFloatFromBool(Bool=$User.bIsFrontEndPreview, True Float=20, False Float=&Multiply_Float(A=60, B=$System.FNLODDistanceRemap))
    )
    Scale RGB=&MakeVectorFromLinearColorRGB(Color=&ColorFromCurve(ColorCurve=DataInterface() {propertyValues {
        AlphaCurve {keys []}
        BlueCurve {keys [
          <at 0 0.4841 RCIM_Cubic>
          <at 0.1789 0.5052>
          {arriveTangent=-0.8205, interpMode=RCIM_Cubic, leaveTangent=-0.8205, time=0.4767, value=1}
          <at 0.7946 0>
          <at 0.984 0.4841>
         ]}
        GreenCurve {keys [
          <at 0 0.0417 RCIM_Cubic>
          <at 0.1789 0.2007>
          {arriveTangent=-0.3052, interpMode=RCIM_Cubic, leaveTangent=-0.3052, time=0.4767, value=0}
          <at 0.7946 0.0129>
          <at 0.984 0.0417>
         ]}
        RedCurve {keys [
          <at 0 1 RCIM_Cubic>
          <at 0.1789 0.0002>
          {arriveTangent=1.6239, interpMode=RCIM_Cubic, leaveTangent=1.6239, time=0.4767, value=0.2174}
          <at 0.7946 1>
          <at 0.984 1>
         ]}
       }}))
    ScaleA=true
    ScaleRGB=true
   ) {module=ScaleColor001}
   &ScaleColor(
    Scale Alpha=&Lerp_Float(
     A=&Multiply_Float(
      A=&Lerp_Float(A=0, B=1.2, Alpha=$System.InterpolateOverTime.InterpolateOverTime_Float.Moving Average)
      B=&ClampFloat(Float=$NPC.FortniteNPC.EventAudio_Amplitude, Min=0.8, Max=1.5)
     )
     B=1.2
     Alpha=$System.MusicReactivity
    )
    ScaleA=true
    ScaleRGB=true
   )
  ]
  rendererClasses [&NiagaraMeshRendererProperties]
  renderers [
   {rendererClass=&NiagaraMeshRendererProperties, rendererIndex=0}
   {rendererClass=&NiagaraMeshRendererProperties, rendererIndex=1}
   {rendererClass=&NiagaraMeshRendererProperties, rendererIndex=2}
  ]
 }
 Heart001 {
  emitterUpdateScript [
   &EmitterState(Life Cycle Mode=::NewEnumerator1, Loop Duration=1, UseLoopDelay=false)
   &SpawnRate(
    Spawn Probability=&MakeCustomFloatFromBool(
     Bool=&SetBoolByFloatComparison(A=$NPC.FortniteNPC.EventAudio_Amplitude, B=0, Comparison Type=::NewEnumerator4)
     True Float=&Lerp_Float(A=0, B=1, Alpha=$System.MusicReactivity)
     False Float=1
    )
    SpawnRate=5
    Use Spawn Probability=true
   )
  ]
  particleSpawnScript [
   &FN_SetSpawnGroup(Number of Spawn Groups=2)
   &InitializeParticle(
    Color=<rgba 1 0.15 0.3595 2>
    Lifetime Max=0.8
    Lifetime Min=0.4
    Lifetime Mode=::NewEnumerator1
    Sprite Size Mode=::NewEnumerator1
    Uniform Sprite Size Max=35
    Uniform Sprite Size Min=25
    UsePositionOffset=false
    Write Material Random=true
   )
   &ShapeLocation(
    MaskSpawnGroup=true
    Non Uniform Scale=<4.5 1 1>
    Offset=&SelectVectorFromArray(Direct Array Index=$Particles.SpawnGroup, Vector Selection Array=DataInterface() {propertyValues {FloatData [<40 0 33>, <40 0 -33>]}})
    Offset Mode=::NewEnumerator0
    Override Seed=false
    Sphere Radius=10
    UseOwnerScale=true
   )
   &CameraOffset(Camera Offset Amount=20)
   &SkeletalMeshLocation(
    Bone / Socket Index=13
    Bone Sampling Mode=::NewEnumerator4
    MaskSpawnGroup=false
    Orientation Sampling=::NewEnumerator1
    Sampled Position Offset=&AddVectorToPosition(Position=Linked() {linkedVariable {name=Particles.ShapeLocation.ShapeVector, type=Vector3f}}, Vector=<0 0 0>)
    Skeletal Mesh=$System.Skeletal Mesh Reader
   )
  ]
  particleUpdateScript [
   &ParticleState()
   &ScaleSpriteSize(
    Scale Sprite Size Mode=::NewEnumerator0
    Uniform Scale Factor=&Multiply_Float(
     A=&Multiply_Float(A=&RandomRangeFloat(Minimum=1.2, Maximum=1.5), B=1)
     B=&FN_LOD_Distance_Remap(Start Distance=240, End Distance=600, Minimum Output=0.7, Maximum Output=1, Clamp Output=true)
    )
   )
   &SubUVAnimation(
    End Frame Range Override=3
    "Sprite Renderer "=DataInterface() {propertyValues {SpriteRenderer=&self:Heart001_0.NiagaraSpriteRendererProperties_0}}
    UseEndFrame=true
    UseStartFrame=true
   )
   &DynamicMaterialParameters(
    Index 0 Param 1=-2
    Index 0 Param 3=1
    Index 0 Param 4=&FloatFromCurve(FloatCurve=DataInterface() {propertyValues {Curve {keys [
         <at 0 2>
         {arriveTangent=-12.6958, leaveTangent=-12.6958, time=0.4, value=2}
         {arriveTangent=-1.5184, leaveTangent=-1.5184, time=0.5, value=0.25}
         <at 0.8 0>
        ]}}})
    Index 1 Param 1=&FloatFromCurve(FloatCurve=DataInterface() {propertyValues {Curve {keys [<at 0 0 RCIM_Constant>,<at 0.0583 0 RCIM_Constant>,<at 0.1173 1 RCIM_Constant>,<at 0.2819 0 RCIM_Constant>,<at 0.3914 1 RCIM_Constant>,<at 0.5142 1 RCIM_Constant>,<at 0.5496 0 RCIM_Constant>,<at 0.672 0 RCIM_Constant>,<at 0.7179 1 RCIM_Constant>,<at 0.8232 1 RCIM_Constant>,<at 0.8524 0 RCIM_Constant>,<at 1 0 RCIM_Constant>]}}})
    Module.Index0Float0Write=true
    Module.Index0Float1Write=true
    Module.Index0Float2Write=true
    Module.Index0Float3Write=true
    Module.Index1Float0Write=true
    Module.Index1Float1Write=true
    Module.Index1Float2Write=true
    Module.Index1Float3Write=true
    Write Parameter Index 1=::NewEnumerator1
   )
   &ScaleColor(
    Scale Alpha=&ClampFloat(
     Float=&FN_LOD_Distance_Remap(Start Distance=200, End Distance=300, Minimum Output=0, Maximum Output=1)
     Max=2
    )
    Scale RGB=&Multiply_VectorByFloat(
     Vector=&MakeVectorFromLinearColorRGB(Color=&ColorFromCurve(ColorCurve=DataInterface() {propertyValues {
         AlphaCurve {keys []}
         BlueCurve {keys [<at 0.1405 1 RCIM_Cubic>,<at 0.6632 3>,<at 1 1>]}
         GreenCurve {keys [<at 0.1405 1 RCIM_Cubic>,<at 0.6632 0.9328>,<at 1 1>]}
         RedCurve {keys [<at 0.1405 1 RCIM_Cubic>,<at 0.6632 0.45>,<at 1 1>]}
        }}))
     Float=&Lerp_Float(A=1, B=1.5, Alpha=$NPC.FortniteNPC.EventAudio_Amplitude)
    )
    ScaleA=true
    ScaleRGB=true
   )
   /SetVariables(Particles.Color002=&MakeLinearColorFromVectorAndFloat(
     "Vector (RGB)"=&MakeVectorFromLinearColorRGB(Color=&ColorFromCurve(ColorCurve=DataInterface() {propertyValues {
         AlphaCurve {keys []}
         BlueCurve {keys [{interpMode=RCIM_Cubic, tangentMode=RCTM_User, time=0.5651, value=0.118}]}
         GreenCurve {keys [{interpMode=RCIM_Cubic, tangentMode=RCTM_User, time=0.5651, value=0.043}]}
         RedCurve {keys [{interpMode=RCIM_Cubic, tangentMode=RCTM_User, time=0.5651, value=0.6875}]}
        }}))
     "Float (Alpha)"=&FloatFromCurve(
      FloatCurve=DataInterface() {propertyValues {Curve {keys [
          <at 0 0>
          {arriveTangent=-2.0518, leaveTangent=-2.0518, time=0.05, value=1}
          {arriveTangent=-1.1111, leaveTangent=-1.1111, time=0.9, value=1}
          <at 1 1>
         ]}}}
      Scale Curve=&ClampFloat(
       Float=&FN_LOD_Distance_Remap(Start Distance=200, End Distance=300, Minimum Output=0, Maximum Output=1)
       Max=1.5
      )
     )
    )) {isSetParameters=true}
   /SetVariables(Particles.Color003:LinearColor=&MultiplyLinearColor(
     Linear Color=<rgba 1 0.151 0.4527 1>
     Scale Factor=&MakeVector4_FromFloat(Float=&Multiply_Float(
       A=&FN_FrontEndFloatModifier(Gameplay=1, Front End=0.8)
       B=&ClampFloat(
        Float=&FN_LOD_Distance_Remap(Start Distance=200, End Distance=300, Minimum Output=0, Maximum Output=1)
        Max=1.5
       )
      ))
    )) {isSetParameters=true}
   /SetVariables(Particles.Size003:Vector2f=&Multiply_Vector2D(A=$Particles.SpriteSize, B=&Vector2DFromFloat(Value=1))) {isSetParameters=true}
  ]
  rendererClasses [&NiagaraSpriteRendererProperties]
  renderers [
   {rendererClass=&NiagaraSpriteRendererProperties}
   {rendererClass=&NiagaraSpriteRendererProperties, rendererIndex=1}
   {rendererClass=&NiagaraSpriteRendererProperties, rendererIndex=2}
   {rendererClass=&NiagaraSpriteRendererProperties, rendererIndex=3}
  ]
 }
 Particle_Line {
  emitterUpdateScript [
   &EmitterState(Life Cycle Mode=::NewEnumerator1, Loop Duration=1, UseLoopDelay=false)
   &SpawnRate(
    Spawn Probability=&MakeCustomFloatFromBool(
     Bool=&SetBoolByFloatComparison(A=$NPC.FortniteNPC.EventAudio_Amplitude, B=0, Comparison Type=::NewEnumerator4)
     True Float=&Lerp_Float(A=0, B=1, Alpha=$System.MusicReactivity)
     False Float=1
    )
    SpawnRate=&FN_FrontEndFloatModifier(Gameplay=10, Front End=20)
    Use Spawn Probability=true
   )
  ]
  particleSpawnScript [
   &PartitionParticles(Number Of Partitions=12)
   &InitializeParticle(
    Color=<rgba 25 25 25 5>
    Lifetime Max=0.6
    Lifetime Min=0.3
    Lifetime Mode=::NewEnumerator1
    Mesh Scale=&MakeCustomVectorFromBool(Bool=&RandomBool(), True Vector=<1 1 1>, False Vector=<-1 -1 -1>)
    Mesh Scale Mode=::NewEnumerator4
    UsePositionOffset=false
    Write Material Random=true
   ) {module=InitializeParticle001}
   &InitialMeshOrientation(
    Mesh Orientation Mode=::NewEnumerator2
    Rotation=&RandomRangeVector(Minimum=<0 0 0>, Maximum=<0 0 1>)
    Use Rotation Vector=true
   )
  ]
  particleUpdateScript [
   &ParticleState()
   &SkeletalMeshLocation(
    Bone / Socket Index=&SelectIntFromArray(
     Direct Array Index=$Particles.PartitionParticles.Partition
     Int Selection Array=DataInterface() {propertyValues {IntData [0,1,2,3,4,5,6,7,8,9,12]}}
    )
    Bone Sampling Mode=::NewEnumerator4
    MaskSpawnGroup=false
    Orientation Sampling=::NewEnumerator1
    Sampled Position Offset=&SelectVectorFromArray(
     Direct Array Index=$Particles.PartitionParticles.Partition
     Vector Selection Array=DataInterface() {propertyValues {FloatData [<0 0 0>, <0 0 0>, <0 0 0>, <0 0 0>, <0 0 0>, <5 0 0>, <-5 0 0>, <0 0 0>, <0 0 0>, <0 0 0>, <0 0 0>]}}
    )
    Skeletal Mesh=$System.Skeletal Mesh Reader
   )
   &ScaleMeshSize(Scale Factor=&RandomRangeVector(Minimum=<0.5 0.5 0.3>, Maximum=<0.6 0.6 0.5>))
   &ScaleMeshSize(Scale Factor=&VectorFromFloat(Value=&RandomRangeFloat(Minimum=0.4, Maximum=0.5))) {module=ScaleMeshSize001}
   &SolveForcesAndVelocity(Clamp Velocity=false, Limit Acceleration=false)
   &ScaleColor(
    Scale Alpha=&FloatFromCurve(
     FloatCurve=DataInterface() {propertyValues {Curve {keys [<at 0 0>,<at 0.1 1>,<at 0.9 1>,<at 1 0>]}}}
     Scale Curve=&Multiply_Float(A=3, B=$System.FNLODDistanceRemap)
    )
    Scale RGB=&MakeVectorFromLinearColorRGB(Color=&ColorFromCurve(ColorCurve=DataInterface() {propertyValues {
        AlphaCurve {keys []}
        BlueCurve {keys [{arriveTangent=2.226, interpMode=RCIM_Cubic, time=0.4568, value=0.4009}, <at 0.645 0.5751 RCIM_Cubic>]}
        GreenCurve {keys [{arriveTangent=-2.4219, interpMode=RCIM_Cubic, time=0.4568, value=0.0615}, <at 0.645 0.0111 RCIM_Cubic>]}
        RedCurve {keys [{arriveTangent=1.9029, interpMode=RCIM_Cubic, time=0.4568, value=1}, <at 0.645 0.0431 RCIM_Cubic>]}
       }}))
    ScaleA=true
    ScaleRGB=true
   )
   &ScaleColor(
    Scale Alpha=&Lerp_Float(
     A=&Multiply_Float(
      A=&Lerp_Float(A=0, B=1, Alpha=$System.InterpolateOverTime.InterpolateOverTime_Float.Moving Average)
      B=&ClampFloat(Float=$NPC.FortniteNPC.EventAudio_Amplitude, Min=0.5)
     )
     B=1
     Alpha=$System.MusicReactivity
    )
    ScaleA=true
    ScaleRGB=true
   ) {module=ScaleColor001}
   &DynamicMaterialParameters(
    Index 0 Param 1=&FloatFromCurve(
     FloatCurve=DataInterface() {propertyValues {Curve {keys [<at 0 -1>,<at 1 1>]}}}
     Scale Curve=&RandomRangeFloat(Minimum=0.7, Maximum=1.1)
    )
    Index 0 Param 2=&RandomRangeFloat(Minimum=0.8, Maximum=1)
    Module.Index0Param0Write=true
    Module.Index0Param1Write=true
    Module.Index0Param2Write=true
    Module.Index0Param3Write=true
   )
   &ScaleMeshSize(Scale Factor=&VectorFromFloat(Value=&MakeCustomFloatFromBool(
      Bool=$User.bisHandHidden
      True Float=&SelectFloatFromArray(
       Direct Array Index=$Particles.PartitionParticles.Partition
       Float Selection Array=DataInterface() {propertyValues {FloatData [1,1,0,0,0,0,1,1,1,1,1,1]}}
      )
      False Float=&SelectFloatFromArray(
       Direct Array Index=$Particles.PartitionParticles.Partition
       Float Selection Array=DataInterface() {propertyValues {FloatData [1,1,1,1,1,1,1,1,1,1,1,1]}}
      )
     ))) {module=ScaleMeshSize002}
   &ScaleMeshSize(Scale Factor=&VectorFromFloat(Value=&MakeCustomFloatFromBool(
      Bool=$User.bisFootHidden
      True Float=&SelectFloatFromArray(
       Direct Array Index=$Particles.PartitionParticles.Partition
       Float Selection Array=DataInterface() {propertyValues {FloatData [1,1,1,1,1,1,1,1,0,0,0,0]}}
      )
      False Float=&SelectFloatFromArray(
       Direct Array Index=$Particles.PartitionParticles.Partition
       Float Selection Array=DataInterface() {propertyValues {FloatData [1,1,1,1,1,1,1,1,1,1,1,1]}}
      )
     ))) {module=ScaleMeshSize003}
  ]
  rendererClasses [&NiagaraMeshRendererProperties]
  renderers [{rendererClass=&NiagaraMeshRendererProperties, rendererIndex=0}]
 }
}
summary {
 systemName=NS_Character_MarchTreatCup_Body
 emitters [
  {emitterName=Aura_small001, rendererClasses [&NiagaraSpriteRendererProperties]}
  {emitterName=Particle_Line, rendererClasses [&NiagaraMeshRendererProperties]}
  {emitterName=Aura_small003, rendererClasses [&NiagaraMeshRendererProperties]}
  {emitterName=Decal, rendererClasses [&NiagaraSpriteRendererProperties]}
  {emitterName=Heart001, rendererClasses [&NiagaraSpriteRendererProperties]}
  {emitterName=Glow_Body, rendererClasses [&NiagaraMeshRendererProperties]}
 ]
 userVariables [
  User.bisFootHidden {type=NiagaraBool}
  User.bIsFrontEnd {type=NiagaraBool}
  User.bIsFrontEndPreview {type=NiagaraBool}
  User.bisHandHidden {type=NiagaraBool}
  User.bisHeadHidden {type=NiagaraBool}
  User.bIsLocalPlayer {type=NiagaraBool, defaultValue=true}
  User.EnableFrontendFakeAudioLoop {type=NiagaraFloat, defaultValue=0, description=""}
 ]
}
