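# Niagara system for the head effect on the MarchTreatCup female character cosmetic: a music-reactive aura with streaking particles.
# A sprite aura emitter and a mesh line emitter reposition particles on the skeletal mesh every update, with an offset.
# Event audio amplitude and MusicReactivity control spawn probability and modulate aura alpha.
# LOD distance fades both emitters, and User.bisHeadHidden scales their particles to zero.
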NiagaraSystem /BRCosmetics/Characters/Player/Female/Medium/Bodies/F_MED_MarchTreatCup/Meshes/FX/NS_Character_MarchTreatCup_head
dependencies {
 usedDataInterfaces [
  &NiagaraDataInterfaceArrayFloat
  &NiagaraDataInterfaceCurve
  &NiagaraDataInterfaceColorCurve
  &NiagaraDataInterfaceSkeletalMesh
 ]
 usedDynamicInputs [&SelectFloatFromArray, &FN_LOD_Distance_Remap]
 usedModules [
  &self:SystemSpawnScript.SystemScriptSource.SystemScriptGraph.NiagaraNodeAssignment_0.SetVariables_#1
  &self:SystemSpawnScript.SystemScriptSource.SystemScriptGraph.NiagaraNodeAssignment_1.SetVariables_#2
  &self:SystemSpawnScript.SystemScriptSource.SystemScriptGraph.NiagaraNodeAssignment_2.SetVariables_#3
  &SystemState
  &InterpolateOverTime
 ]
}
emitters {
 Aura_small001 {
  emitterUpdateScript [
   &EmitterState(Life Cycle Mode=::NewEnumerator1, Loop Duration=1, UseLoopDelay=false)
   &SpawnRate(
    Spawn Probability=&Lerp_Float(
     A=&MakeCustomFloatFromBool(
      Bool=&SetBoolByFloatComparison(A=$NPC.FortniteNPC.EventAudio_Amplitude, B=0, Comparison Type=::NewEnumerator4)
      True Float=0
      False Float=1
     )
     B=1
     Alpha=$System.MusicReactivity
    )
    SpawnRate=2
    Use Spawn Probability=true
   )
  ]
  particleSpawnScript [
   &InitializeParticle(
    Color=<rgba 0.6771 1.1181 2 3>
    Lifetime=2
    Sprite Size Max=<60 40>
    Sprite Size Min=<30 30>
    Sprite Size Mode=::NewEnumerator2
    Sprite UV Mode=::NewEnumerator1
    UsePositionOffset=false
    Write Material Random=true
   )
   &CameraOffset(Camera Offset Amount=5)
   &ShapeLocation(MaskSpawnGroup=false, Override Seed=false, Sphere Radius=5, UseOwnerScale=true)
  ]
  particleUpdateScript [
   &ParticleState()
   &SkeletalMeshLocation(
    Bone Sampling Mode=::NewEnumerator4
    MaskSpawnGroup=false
    Sampled Position Offset=&RandomRangeVector(Minimum=<20 -5 -5>, Maximum=<20 5 5>)
    Skeletal Mesh=$System.Skeletal Mesh Reader
   )
   &ScaleSpriteSize(
    Scale Factor=&RandomRangeVector2D(Minimum=<0.5 0.6>, Maximum=<0.55 0.65>)
    Scale Sprite Size Mode=::NewEnumerator2
   ) {module=ScaleSpriteSize001}
   &ScaleColor(Scale Alpha=$System.FNLODDistanceRemap, ScaleA=true, ScaleRGB=true) {module=ScaleColor001}
   &ScaleColor(
    Scale Alpha=&FloatFromCurve(
     FloatCurve=DataInterface() {propertyValues {Curve {keys [
         {tangentMode=RCTM_User, time=0, value=0}
         <at 0.2 1>
         {arriveTangent=-0.0405, interpMode=RCIM_Cubic, leaveTangent=-0.0405, tangentMode=RCTM_User, time=0.4, value=1}
         {arriveTangent=-1.4489, interpMode=RCIM_Cubic, leaveTangent=-1.4489, tangentMode=RCTM_User, time=1, value=0}
        ]}}}
     Scale Curve=&Lerp_Float(
      A=&Multiply_Float(
       A=&Lerp_Float(A=0, B=1, Alpha=$System.InterpolateOverTime.InterpolateOverTime_Float.Moving Average)
       B=&ClampFloat(Float=$NPC.FortniteNPC.EventAudio_Amplitude, Min=0.9, Max=1.5)
      )
      B=1
      Alpha=$System.MusicReactivity
     )
    )
    Scale RGB=&MakeVectorFromLinearColorRGB(Color=<rgba 10 2.8125 4.987 1>)
    ScaleA=true
    ScaleRGB=true
   )
   &DynamicMaterialParameters(
    Index 0 Param 1=&RandomRangeFloat(Minimum=0.5, Maximum=1.5)
    Module.Index0Float0Write=true
    Module.Index0Float1Write=true
    Module.Index0Float2Write=true
    Module.Index0Float3Write=true
   )
   &ScaleSpriteSize(
    Scale Sprite Size Mode=::NewEnumerator0
    Uniform Scale Factor=&MakeCustomFloatFromBool(Bool=$User.bisHeadHidden, True Float=0, False Float=1)
   )
  ]
  rendererClasses [&NiagaraSpriteRendererProperties]
  renderers [{rendererClass=&NiagaraSpriteRendererProperties}]
 }
 Particle_Line {
  emitterUpdateScript [
   &EmitterState(Life Cycle Mode=::NewEnumerator1, Loop Duration=1, UseLoopDelay=false)
   &SpawnRate(
    Spawn Probability=&MakeCustomFloatFromBool(
     Bool=&SetBoolByFloatComparison(A=$NPC.FortniteNPC.EventAudio_Amplitude, B=0, Comparison Type=::NewEnumerator4)
     True Float=&Lerp_Float(A=0, B=1, Alpha=$System.MusicReactivity)
     False Float=1
    )
    Use Spawn Probability=true
   )
  ]
  particleSpawnScript [
   &InitializeParticle(
    Color=<rgba 20 20 20 5>
    Lifetime=0.6
    Mesh Scale=&MakeCustomVectorFromBool(Bool=&RandomBool(), True Vector=<1 1 1>, False Vector=<-1 -1 -1>)
    Mesh Scale Mode=::NewEnumerator4
    UsePositionOffset=false
    Write Material Random=true
   ) {module=InitializeParticle001}
   &InitialMeshOrientation(
    Mesh Orientation Mode=::NewEnumerator2
    Rotation=&RandomRangeVector(Minimum=<0 0 0>, Maximum=<1 0 0>)
    Use Rotation Vector=true
   )
   &CameraOffset(Camera Offset Amount=-5)
  ]
  particleUpdateScript [
   &ParticleState()
   &SkeletalMeshLocation(Bone Sampling Mode=::NewEnumerator4, MaskSpawnGroup=false, Sampled Position Offset=<20 0 0>, Skeletal Mesh=$System.Skeletal Mesh Reader)
   &ScaleMeshSize(Scale Factor=&RandomRangeVector(Minimum=<0.5 0.5 0.3>, Maximum=<0.6 0.6 0.5>))
   &ScaleMeshSize(Scale Factor=&VectorFromFloat(Value=&RandomRangeFloat(Minimum=0.3, Maximum=0.4))) {module=ScaleMeshSize001}
   &SolveForcesAndVelocity(Clamp Velocity=false, Limit Acceleration=false)
   &ScaleColor(
    Scale Alpha=&FloatFromCurve(
     FloatCurve=DataInterface() {propertyValues {Curve {keys [<at 0 0>,<at 0.1 1>,<at 0.9 1>,<at 1 0>]}}}
     Scale Curve=&Multiply_Float(A=1, B=$System.FNLODDistanceRemap)
    )
    Scale RGB=&MakeVectorFromLinearColorRGB(Color=&ColorFromCurve(ColorCurve=DataInterface() {propertyValues {
        AlphaCurve {keys []}
        BlueCurve {keys [{arriveTangent=2.226, interpMode=RCIM_Cubic, time=0.397, value=0.8143}, <at 0.6226 2.0833 RCIM_Cubic>]}
        GreenCurve {keys [{arriveTangent=-2.4219, interpMode=RCIM_Cubic, time=0.397, value=0.1249}, <at 0.6226 0.0401 RCIM_Cubic>]}
        RedCurve {keys [{arriveTangent=1.9029, interpMode=RCIM_Cubic, time=0.397, value=2.0312}, <at 0.6226 0.1562 RCIM_Cubic>]}
       }}))
    ScaleA=true
    ScaleRGB=true
   )
   &DynamicMaterialParameters(
    Index 0 Param 1=&FloatFromCurve(
     FloatCurve=DataInterface() {propertyValues {Curve {keys [<at 0 -1>,<at 1 1>]}}}
     Scale Curve=&RandomRangeFloat(Minimum=0.7, Maximum=1.1)
    )
    Index 0 Param 2=&RandomRangeFloat(Minimum=0.8, Maximum=1)
    Module.Index0Param0Write=true
    Module.Index0Param1Write=true
    Module.Index0Param2Write=true
    Module.Index0Param3Write=true
   )
   &ScaleMeshSize(Scale Factor=&VectorFromFloat(Value=&MakeCustomFloatFromBool(Bool=$User.bisHeadHidden, True Float=0, False Float=1))) {module=ScaleMeshSize002}
  ]
  rendererClasses [&NiagaraMeshRendererProperties]
  renderers [{rendererClass=&NiagaraMeshRendererProperties, rendererIndex=0}]
 }
}
summary {
 systemName=NS_Character_MarchTreatCup_head
 emitters [
  {emitterName=Aura_small001, rendererClasses [&NiagaraSpriteRendererProperties]}
  {emitterName=Particle_Line, rendererClasses [&NiagaraMeshRendererProperties]}
 ]
 userVariables [
  User.bIsFrontEnd {type=NiagaraBool}
  User.bIsFrontEndPreview {type=NiagaraBool}
  User.bisHeadHidden {type=NiagaraBool}
  User.bIsLocalPlayer {type=NiagaraBool, defaultValue=true}
 ]
}
