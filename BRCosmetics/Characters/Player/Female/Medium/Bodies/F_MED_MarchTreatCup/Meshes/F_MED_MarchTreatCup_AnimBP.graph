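# Animation Blueprint for the female medium MarchTreatCup cosmetic body, adding secondary motion to its bows and blouse.
# It copies the attached parent mesh's pose, runs the body's Control Rig and caches the result.
# Chained spring-driven AnimDynamics nodes simulate left and right bow and blouse bones in root-relative space with reduced gravity.
# MSTR_Bow then follows a weighted spine, pectoral and clavicle constraint, with an additive offset blended in while bIsTargeting is true.
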AnimBlueprint /BRCosmetics/Characters/Player/Female/Medium/Bodies/F_MED_MarchTreatCup/Meshes/F_MED_MarchTreatCup_AnimBP
target_skeleton=&/F_MED_MarchTreatCup_Skeleton
graphs [
 AnimGraph {nodes [
   n1 = Anim Dynamics - Bone: dyn_bow_01_l(ComponentPose=Local To Component(LocalPose=Use cached pose 'Control Rig'()), Alpha=1, ExternalForce="0,0,0") {
    angularDampingOverride=0.999
    angularSpringConstant=560
    bAngularSpring=true
    bOverrideAngularDamping=true
    bUseGravityOverride=true
    bUsePlanarLimit=false
    gravityScale=0.35
    simulationSpace=RootRelative
    boundBone {boneName=dyn_bow_01_l}
    componentLinearAccScale=<1 1 1>
    componentLinearVelScale=<1 1 1>
    gravityOverride {z=-980}
    physicsBodyDefinitions [{
      sphereCollisionRadius=2
      boundBone {boneName=dyn_bow_01_l}
      boxExtents=<6 11 6>
      constraintSetup {angularLimitsMax=<20 15 0>, angularLimitsMin=<5 -15 0>, angularTarget=<1 0 0>}
      localJointOffset=<0 0 0>
     }]
   }
   n2 = Anim Dynamics - Bone: dyn_bow_02_l(ComponentPose=@n1, Alpha=1, ExternalForce="0,0,0") {
    angularDampingOverride=0.999
    angularSpringConstant=560
    bAngularSpring=true
    bOverrideAngularDamping=true
    bUseGravityOverride=true
    gravityScale=0.35
    simulationSpace=RootRelative
    boundBone {boneName=dyn_bow_02_l}
    componentLinearAccScale=<1 1 1>
    componentLinearVelScale=<1 1 1>
    gravityOverride {z=-980}
    physicsBodyDefinitions [{
      boundBone {boneName=dyn_bow_02_l}
      boxExtents=<6 6 4>
      constraintSetup {angularLimitsMax=<0 15 15>, angularLimitsMin=<0 -15 0>, angularTarget=<1 0 0>}
      localJointOffset=<0 0 0>
     }]
   }
   n3 = Anim Dynamics - Bone: dyn_bow_01_r(ComponentPose=@n2, Alpha=1, ExternalForce="0,0,0") {
    angularDampingOverride=0.999
    angularSpringConstant=560
    bAngularSpring=true
    bOverrideAngularDamping=true
    bUseGravityOverride=true
    gravityScale=0.35
    simulationSpace=RootRelative
    boundBone {boneName=dyn_bow_01_r}
    componentLinearAccScale=<1 1 1>
    componentLinearVelScale=<1 1 1>
    gravityOverride {z=-980}
    physicsBodyDefinitions [{
      boundBone {boneName=dyn_bow_01_r}
      boxExtents=<6 11 6>
      constraintSetup {angularLimitsMax=<0 15 0>, angularLimitsMin=<-20 -15 0>, angularTarget=<1 0 0>}
      localJointOffset=<0 0 0>
     }]
   }
   n4 = Anim Dynamics - Bone: dyn_bow_02_r(ComponentPose=@n3, Alpha=1, ExternalForce="0,0,0") {
    angularDampingOverride=0.999
    angularSpringConstant=560
    bAngularSpring=true
    bOverrideAngularDamping=true
    bUseGravityOverride=true
    gravityScale=0.35
    simulationSpace=RootRelative
    boundBone {boneName=dyn_bow_02_r}
    componentLinearAccScale=<1 1 1>
    componentLinearVelScale=<1 1 1>
    gravityOverride {z=-980}
    physicsBodyDefinitions [{
      boundBone {boneName=dyn_bow_02_r}
      boxExtents=<6 6 4>
      constraintSetup {angularLimitsMax=<0 15 15>, angularLimitsMin=<0 -15 0>, angularTarget=<1 0 0>}
      localJointOffset=<0 0 0>
     }]
   }
   n5 = Anim Dynamics - Bone: dyn_black_bluse_l(ComponentPose=@n4, Alpha=1, ExternalForce="0,0,0") {
    angularDampingOverride=1
    angularSpringConstant=5000
    bAngularSpring=true
    bOverrideAngularDamping=true
    bUseGravityOverride=true
    gravityScale=0.35
    simulationSpace=RootRelative
    boundBone {boneName=dyn_black_bluse_l}
    componentLinearAccScale=<1 1 1>
    componentLinearVelScale=<1 1 1>
    gravityOverride {z=-980}
    physicsBodyDefinitions [{
      boundBone {boneName=dyn_black_bluse_l}
      boxExtents=<20 25 20>
      constraintSetup {angularLimitsMax=<15 15 5>, angularLimitsMin=<-15 -15 -45>, angularTarget=<1 0 0>}
      localJointOffset=<3 0 0>
     }]
   }
   n6 = Anim Dynamics - Bone: dyn_black_bluse_r(ComponentPose=@n5, Alpha=1, ExternalForce="0,0,0") {
    angularDampingOverride=1
    angularSpringConstant=5000
    bAngularSpring=true
    bOverrideAngularDamping=true
    bUseGravityOverride=true
    gravityScale=0.35
    simulationSpace=RootRelative
    boundBone {boneName=dyn_black_bluse_r}
    componentLinearAccScale=<1 1 1>
    componentLinearVelScale=<1 1 1>
    gravityOverride {z=-980}
    physicsBodyDefinitions [{
      boundBone {boneName=dyn_black_bluse_r}
      boxExtents=<25 25 10>
      constraintSetup {angularLimitsMax=<15 15 5>, angularLimitsMin=<-15 -15 -45>, angularTarget=<1 0 0>}
      localJointOffset=<3 0 0>
     }]
   }
   n7 = Anim Dynamics - Bone: dyn_bluse_01_l1(ComponentPose=@n6, Alpha=1, ExternalForce="0,0,0") {
    angularDampingOverride=0.9999
    angularSpringConstant=5000
    bAngularSpring=true
    bOverrideAngularDamping=true
    bUseGravityOverride=true
    gravityScale=0.35
    simulationSpace=RootRelative
    boundBone {boneName=dyn_bluse_01_l1}
    componentAppliedLinearAccClamp=<3000 3000 3000>
    componentLinearAccScale=<1 1 1>
    componentLinearVelScale=<1 1 1>
    gravityOverride {z=-980}
    physicsBodyDefinitions [{
      boundBone {boneName=dyn_bluse_01_l1}
      boxExtents=<14 14 14>
      constraintSetup {angularLimitsMax=<15 15 25>, angularLimitsMin=<-15 -15 -2>, angularTarget=<1 0 0>}
      localJointOffset=<-1 0 0>
     }]
   }
   n8 = Anim Dynamics - Bone: dyn_bluse_01_r(ComponentPose=@n7, Alpha=1, ExternalForce="0,0,0") {
    angularDampingOverride=0.9999
    angularSpringConstant=5000
    bAngularSpring=true
    bOverrideAngularDamping=true
    bUseGravityOverride=true
    gravityScale=0.35
    simulationSpace=RootRelative
    boundBone {boneName=dyn_bluse_01_r}
    componentAppliedLinearAccClamp=<3000 3000 3000>
    componentLinearAccScale=<1 1 1>
    componentLinearVelScale=<1 1 1>
    gravityOverride {z=-980}
    physicsBodyDefinitions [{
      boundBone {boneName=dyn_bluse_01_r}
      boxExtents=<14 14 14>
      constraintSetup {angularLimitsMax=<15 15 25>, angularLimitsMin=<-15 -15 -2>, angularTarget=<1 0 0>}
      localJointOffset=<-1 0 0>
     }]
   }
   n9 = "Constraint - MSTR_Bow (PPPP)"!(
    ComponentPose=@n8
    Alpha=1
    ConstraintWeights_0=0.5175
    ConstraintWeights_1:float=0.462
    ConstraintWeights_2:float=0.0192
    ConstraintWeights_3:float=0.0013
   ) -> (Pose) {
    boneToModify {boneName=MSTR_Bow}
    constraintSetup [{targetBone {boneName=spine_04}}, {targetBone {boneName=spine_05}}, {targetBone {boneName=deform_pec_l}}, {targetBone {boneName=clavicle_l}}]
    constraintWeights [1,1,1,1]
   }
   n10 = "Transform (Modify) Bone - Bone: MSTR_Bow"!(ComponentPose=@n9, bAlphaBoolEnabled:bool=Get bIsTargeting(), Translation="0,0.7,0.5", Rotation="0.006,0.000,0.895", Scale="1,1,1") -> (Pose) {
    alphaInputType=Bool
    rotationMode=BMM_Additive
    rotationSpace=BCS_BoneSpace
    translationMode=BMM_Additive
    translationSpace=BCS_BoneSpace
    alphaBoolBlend {blendInTime=0.15, blendOutTime=0.5}
    boneToModify {boneName=MSTR_Bow}
    rotation=<rot 0.006 0 0.8949>
    translation {y=0.7, z=0.5}
   }
   Output Pose(Result=Component To Local(ComponentPose=@n10)) {layerGroup=None}
   n11 = Control Rig(Source=Copy Pose From Mesh() {bUseAttachedParent=true}) {controlRigAssetReference {blueprintRigClass=&/F_MED_MarchTreatCup_CtrlRig.F_MED_MarchTreatCup_CtrlRig_C}}
   Save cached pose 'Control Rig'(Pose=@n11) {cacheName=Control Rig}
  ]}
 EventGraph {nodes [Event BlueprintUpdateAnimation()]}
]
variables []
